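# Physics asset for the Prisoner face accessory cosmetic, defining capsule bodies for a three-bone dynamic chain attached to the head.
# The dyn_head_chain bones are constrained in series from the head, each joint sharing the same swing and twist limits.
# Capsules on head, both neck bones and spine_05 give the upper body collision shapes, linked by constraints with free swing and twist.
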PhysicsAsset /BRCosmetics/Accessories/Hats/M_MED_Prisoner_FaceAcc/Meshes/M_MED_Prisoner_FaceAcc_Physics
bodies {
 dyn_head_chain_01 [{
   shapeType=Capsule
   length=2.9361
   radius=2.5181
   shapeName=dyn_head_chain_01_capsule
   center=<1.9105 0.0675 -0.1331>
   rotation=<rot 0 -89.1668 -88.2758>
  }]
 dyn_head_chain_02 [{
   shapeType=Capsule
   length=3.0002
   radius=2.4958
   shapeName=dyn_head_chain_02_capsule
   center=<2.1666 0.3006 -0.1155>
   rotation=<rot -0.0591 89.1358 -89.6651>
  }]
 dyn_head_chain_03 [{
   shapeType=Capsule
   length=1.5123
   radius=3.7468
   shapeName=dyn_head_chain_03_capsule
   center=<2.5916 0.1967 -0.0147>
   rotation=<rot -0.2449 -89.9997 -2.2488>
  }]
 head [{shapeType=Capsule, length=9.2454, radius=9.6426, center=<8.3744 2.6116 -0.078>, rotation=<rot 1.0028 95.4692 -90>}]
 neck_01 [{shapeType=Capsule, length=3.6285, radius=13.7148, center=<-3.1064 3.7281 0.0379>, rotation=<rot 0 0 -1.1598>}]
 neck_02 [{shapeType=Capsule, length=1.3315, radius=6.4933, center=<0.4472 -1.4086 -0.0002>, rotation=<rot 90 0 -12.1454>}]
 spine_05 [{shapeType=Capsule, length=10.2481, radius=16.8443, center=<1.617 2.8685 0.0205>, rotation=<rot 0 0 -1.1597>}]
}
constraints [
 {bone1Name=dyn_head_chain_01, bone2Name=head, swing1LimitDegrees=75, swing2LimitDegrees=75, twistLimitDegrees=30}
 {bone1Name=dyn_head_chain_02, bone2Name=dyn_head_chain_01, swing1LimitDegrees=75, swing2LimitDegrees=75, twistLimitDegrees=30}
 {bone1Name=dyn_head_chain_03, bone2Name=dyn_head_chain_02, swing1LimitDegrees=75, swing2LimitDegrees=75, twistLimitDegrees=30}
 {
  bone1Name=head
  bone2Name=neck_02
  swing1LimitDegrees=45
  swing1Motion=Free
  swing2LimitDegrees=45
  swing2Motion=Free
  twistLimitDegrees=45
  twistMotion=Free
 }
 {
  bone1Name=neck_02
  bone2Name=neck_01
  swing1LimitDegrees=45
  swing1Motion=Free
  swing2LimitDegrees=45
  swing2Motion=Free
  twistLimitDegrees=45
  twistMotion=Free
 }
 {
  bone1Name=neck_01
  bone2Name=spine_05
  swing1LimitDegrees=45
  swing1Motion=Free
  swing2LimitDegrees=45
  swing2Motion=Free
  twistLimitDegrees=45
  twistMotion=Free
 }
]
