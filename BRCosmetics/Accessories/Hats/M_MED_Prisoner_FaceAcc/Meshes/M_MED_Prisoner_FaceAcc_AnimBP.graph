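# Post-attach AnimBlueprint for the Prisoner face accessory cosmetic that follows the character mesh and adds secondary physics motion to its head chain.
# Copy Pose From Mesh reads the attached parent, then feeds a RigidBody node simulating the accessory's physics asset in spine_03 base-bone space.
# Three chained AnimDynamics nodes then drive dyn_head_chain_01 to 03 with angular springs, limits and overridden gravity.
# The event graph has no logic.
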AnimBlueprint /BRCosmetics/Accessories/Hats/M_MED_Prisoner_FaceAcc/Meshes/M_MED_Prisoner_FaceAcc_AnimBP
target_skeleton=&/M_MED_Prisoner_FaceAcc_Skeleton
graphs [
 AnimGraph {nodes [
   n1 = RigidBody(ComponentPose=Local To Component(LocalPose=Copy Pose From Mesh() {bUseAttachedParent=true})) {
    bOverrideWorldGravity=true
    overridePhysicsAsset=&/M_MED_Prisoner_FaceAcc_Physics
    simulationSpace=BaseBoneSpace
    alphaBoolBlend {blendOption=Linear}
    baseBoneRef {boneName=spine_03}
    componentAppliedLinearAccClamp=<3000 3000 3000>
    componentLinearAccScale=<0.5 0.3 0.3>
    componentLinearVelScale=<0.5 0.3 0.3>
    overrideWorldGravity {z=-980}
   }
   n2 = Anim Dynamics - Bone: dyn_head_chain_01(ComponentPose=@n1, Alpha=1, ExternalForce="0,0,0") {
    angularDampingOverride=0.9999
    angularSpringConstant=10
    bAngularSpring=true
    bGravityOverrideInSimSpace=true
    bOverrideAngularDamping=true
    bUseGravityOverride=true
    alphaBoolBlend {blendOption=Linear}
    boundBone {boneName=dyn_head_chain_01}
    componentLinearAccScale=<5 5 5>
    componentLinearVelScale=<5 5 5>
    gravityOverride {z=-980}
    physicsBodyDefinitions [{
      sphereCollisionRadius=0
      boundBone {boneName=dyn_head_chain_01}
      boxExtents=<10 20 20>
      constraintSetup {angularLimitsMax=<10 7 10>, angularLimitsMin=<-10 -60 -10>, angularTarget=<0 0 0>}
      localJointOffset=<2 0 0>
     }]
   }
   n3 = Anim Dynamics - Bone: dyn_head_chain_02(ComponentPose=@n2, Alpha=1, ExternalForce="0,0,0") {
    angularDampingOverride=0.9999
    angularSpringConstant=10
    bAngularSpring=true
    bGravityOverrideInSimSpace=true
    bOverrideAngularDamping=true
    bUseGravityOverride=true
    alphaBoolBlend {blendOption=Linear}
    boundBone {boneName=dyn_head_chain_02}
    componentLinearAccScale=<5 5 5>
    componentLinearVelScale=<5 1 1>
    gravityOverride {z=-980}
    physicsBodyDefinitions [{
      sphereCollisionRadius=0
      boundBone {boneName=dyn_head_chain_02}
      boxExtents=<10 20 20>
      constraintSetup {angularLimitsMax=<10 60 60>, angularLimitsMin=<-10 -60 20>, angularTarget=<0 0 0>}
      localJointOffset=<2 0 0>
     }]
   }
   n4 = Anim Dynamics - Bone: dyn_head_chain_03(ComponentPose=@n3, Alpha=1, ExternalForce="0,0,0") {
    angularDampingOverride=0.9999
    angularSpringConstant=10
    bAngularSpring=true
    bGravityOverrideInSimSpace=true
    bOverrideAngularDamping=true
    bUseGravityOverride=true
    alphaBoolBlend {blendOption=Linear}
    boundBone {boneName=dyn_head_chain_03}
    componentLinearAccScale=<5 5 5>
    componentLinearVelScale=<5 1 1>
    gravityOverride {z=-980}
    physicsBodyDefinitions [{
      sphereCollisionRadius=0
      boundBone {boneName=dyn_head_chain_03}
      boxExtents=<10 20 20>
      constraintSetup {angularLimitsMax=<10 7 80>, angularLimitsMin=<-10 -60 20>, angularTarget=<0 0 0>}
      localJointOffset=<2 0 0>
     }]
   }
   Output Pose(Result=Component To Local(ComponentPose=@n4)) {layerGroup=None}
  ]}
 EventGraph {nodes [
   TryGetPawnOwner()
   Event BlueprintUpdateAnimation()
  ]}
]
variables []
